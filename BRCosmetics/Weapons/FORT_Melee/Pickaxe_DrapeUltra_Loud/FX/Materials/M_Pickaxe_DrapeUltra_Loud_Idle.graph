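# Translucent unlit Niagara ribbon material for the idle effect on the DrapeUltra Loud pickaxe.
# Panning directional noise distorts one color-curve lookup, while two multiplied, scrolling fire-trail textures drive a second; both sum into emissive.
# Named reroutes supply power-warped UVs and wispy-noise distortion, panned in opposite directions on each half of V.
# Opacity combines eroded curve alpha and fire with dynamic-parameter masking, sphere mask, view-angle fade and depth fade.
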Material /BRCosmetics/Weapons/FORT_Melee/Pickaxe_DrapeUltra_Loud/FX/Materials/M_Pickaxe_DrapeUltra_Loud_Idle
graph {
 properties {
  bAutomaticallySetUsageInEditor=false
  bUsedWithNiagaraRibbons=true
  bUsedWithStaticMesh=false
  blendMode=BLEND_Translucent
  shadingModel=MSM_Unlit
  usedShadingModels=Unlit
 }
 nodes [
  n1 = MaterialFunctionCall(XY=TextureCoordinate()) {materialFunction=&Swizzle}
  n2 = Add(A=NamedRerouteUsage(), B=NamedRerouteUsage())
  n3 = Multiply(A=Constant() {r=0.5}, B=ScalarParameter() {defaultValue=0.7, parameterName=U_Tile})
  n4 = Constant() {r=1}
  n5 = ScalarParameter() {defaultValue=1.5, parameterName=PanSpeed}
  n6 = Constant() {r=-1.6}
  n7 = Panner(
   Coordinate=Multiply(
    A=MaterialFunctionCall(XY=@n2) {materialFunction=&Swizzle}.YX
    B=AppendVector(A=@n3, B=@n4)
   )
   Speed=Multiply(A=Constant2Vector() {g=-0.3}, B=Multiply(A=@n5, B=@n6))
  ) {speedX=-0.02, speedY=-0.3}
  n8 = Constant() {r=2}
  n9 = Constant() {r=7}
  n10 = MaterialFunctionCall() {materialFunction=&LinearGradient}
  n11 = Saturate(None=Sine(None=Divide(A=@n10.VGradient)))
  n12 = ComponentMask(None=Divide(
    A=TextureSample(UVs=@n7) {samplerType=SAMPLERTYPE_Normal, texture=/BRCosmetics/Effects/Fort_Effects/Emotes/UltraLight/Texture/T_DirectionalNoise_Normal}.RGB
    B=LinearInterpolate(A=@n8, B=@n9, Alpha=@n11) {constA=1, constB=4}
   )) {g=true, r=true}
  n13 = CurveAtlasRowParameter(CurveTime=MaterialFunctionCall(UV Channel=Add(A=@n1.YX, B=@n12)) {materialFunction=&LinearGradient}.UGradient) {
   atlas=/BRCosmetics/Weapons/FORT_Melee/Pickaxe_DrapeUltra_Loud/FX/ColorCurves/CA_Pickaxe_DrapeUltra_Loud_Idle
   curve=/BRCosmetics/Weapons/FORT_Melee/Pickaxe_DrapeUltra_Loud/FX/ColorCurves/CV_Pickaxe_DrapeUltra_Idle_01
   parameterName=Param_3
  }
  n14 = DynamicParameter() {defaultValue=<rgba 0 1 1 1>, paramNames [Mask,Param2,Param3,Param4]}
  n15 = Multiply(A=@n13, B=Multiply(A=Constant() {r=1}, B=@n14.Param2))
  n16 = Add(A=NamedRerouteUsage(), B=NamedRerouteUsage())
  n17 = Constant() {r=3}
  n18 = ParticleRandom()
  n19 = AppendVector(A=Multiply(A=Add(A=@n17, B=@n18), B=ScalarParameter() {defaultValue=0.7, parameterName=U_Tile}), B=Constant() {r=1})
  n20 = MaterialFunctionCall(UV Channel=MaterialFunctionCall(UVs=Multiply(A=@n16, B=@n19), Texture Scale=Constant2Vector() {g=1.2, r=1}) {materialFunction=&ScaleUVsByCenter}.UVs) {materialFunction=&LinearGradient}
  n21 = Multiply(A=OneMinus(None=@n20.UGradient)) {constB=0.9}
  n22 = Multiply(A=Time(), B=Constant() {r=2.5})
  n23 = TextureSample(UVs=Panner(Coordinate=AppendVector(A=@n21, B=@n20.VGradient), Time=@n22) {speedX=-0.15}) {texture=/Game/Effects/Fort_Effects/Textures/Fire/T_Stylized_Fire_Trail}
  n24 = MaterialFunctionCall(UV Channel=MaterialFunctionCall(UVs=Multiply(A=@n16), Texture Scale=Constant2Vector() {g=1.5, r=0.7}) {materialFunction=&ScaleUVsByCenter}.UVs) {materialFunction=&LinearGradient}
  n25 = OneMinus(None=@n24.UGradient)
  n26 = TextureSample(UVs=Panner(Coordinate=AppendVector(A=@n25, B=@n24.VGradient), Time=@n22) {speedX=-0.35}) {texture=&T_Stylized_Fire_Trail}
  n27 = Multiply(A=@n23.R, B=@n26.R)
  n28 = CurveAtlasRowParameter(CurveTime=@n27) {
   atlas=&CA_Pickaxe_DrapeUltra_Loud_Idle
   curve=/BRCosmetics/Weapons/FORT_Melee/Pickaxe_DrapeUltra_Loud/FX/ColorCurves/CV_Pickaxe_DrapeUltra_Idle_02
   defaultValue=1
   parameterName=Param_2
  }
  n29 = Multiply(A=@n28)
  MaterialFunctionCall(Emissive=Multiply(A=Add(A=@n15, B=@n29), B=ScalarParameter() {defaultValue=1.2, parameterName=OverallEmissive})) {materialFunction=&MF_FX_EmissiveAdjustments}
  MaterialFunctionCall(Offset Amount=Constant()) {materialFunction=/Engine/Functions/Engine_MaterialFunctions02/WorldPositionOffset/CameraOffset}
  Max(A=@n23.R, B=@n26.R)
  n30 = Add(
   A=DynamicParameter() {defaultValue=<rgba 0 1 1 1>, paramNames [Mask,Param2,Param3,Param4]}.Mask
   B=ScalarParameter() {parameterName=DebugMask, sliderMax=1}
  )
  n31 = Add(A=ComponentMask(None=TextureCoordinate()) {r=true}, B=@n30)
  n32 = DotProduct(A=ObjectOrientation(), B=CameraVectorWS())
  n33 = ScalarParameter() {defaultValue=90, parameterName=AngleFade}
  n34 = Multiply(A=OneMinus(None=Abs(None=@n32)), B=@n33)
  n35 = Saturate(None=LinearInterpolate(
    A=Multiply(A=@n34, B=@n34)
    B=MaterialFunctionCall("FadeSensitivity (optional)"=@n33) {materialFunction=&FadeBasedOnViewAngle}.Result
    Alpha=ScalarParameter() {parameterName=InvertAngleFade}
   ))
  n36 = ScalarParameter() {defaultValue=3, parameterName=OveralOpacity}
  n37 = Divide(A=MaterialFunctionCall() {materialFunction=&LinearGradient}.VGradient)
  n38 = Constant() {r=2000}
  n39 = Multiply(
   A=Saturate(None=Sine(None=@n37))
   B=LinearInterpolate(
    A=Constant() {r=5}
    B=Constant() {r=100}
    Alpha=MaterialFunctionCall(Distance=@n38) {materialFunction=&Near_Camera_Fade}.Result
   ) {constA=2, constB=100}
  ) {constB=2}
  n40 = Subtract(A=Multiply(A=@n13.A), B=OneMinus(None=Saturate(None=@n39))) {constB=0.1}
  n41 = MaterialFunctionCall() {materialFunction=&LinearGradient}
  n42 = LinearInterpolate(Alpha=@n14.Mask) {constA=10, constB=0}
  n43 = Subtract(A=Subtract(A=Saturate(None=@n40), B=Multiply(A=@n41.UGradient, B=@n42)) {constB=0.1}, B=Constant() {r=0.08}) {constB=0.1}
  n44 = Multiply(A=Power(Base=Saturate(None=@n43), Exp=Constant() {r=8}), B=Multiply(A=Constant() {r=1}, B=ParticleColor().A))
  n45 = ScalarParameter() {defaultValue=0.8, parameterName=AddSoftDetailOpacity}
  n46 = Constant() {r=1}
  n47 = Add(A=Multiply(A=Constant() {r=2}, B=NamedRerouteUsage()), B=TextureCoordinate())
  n48 = AppendVector(A=Constant() {r=1}, B=Constant() {r=1})
  n49 = Multiply(
   A=Add(A=LinearInterpolate(B=@n44, Alpha=@n45), B=Power(Base=@n27, Exp=@n46)) {constB=0}
   B=SphereMask(
    A=MaterialFunctionCall(UVs=@n47, Texture Scale=@n48) {materialFunction=&ScaleUVsByCenter}.UVs
    B=Constant() {r=0.5}
    Radius=Constant() {r=0.47}
   ) {hardnessPercent=50}
  )
  Multiply(
   A=Saturate(None=SmoothStep(Value=@n31) {constMax=0.1})
   B=Multiply(A=Multiply(A=@n35, B=@n36), B=DepthFade(Opacity=@n49) {fadeDistanceDefault=20})
  )
  n50 = Constant() {r=4}
  n51 = ScalarParameter() {defaultValue=0.7, parameterName=U_Tile}
  n52 = Multiply(A=NamedRerouteUsage(), B=AppendVector(A=Multiply(A=@n50, B=@n51), B=Constant() {r=1}))
  n53 = Constant() {r=0.5}
  n54 = Constant() {r=0.2}
  n55 = Constant() {r=0.5}
  n56 = Panner(Coordinate=@n52, Speed=Divide(A=AppendVector(A=@n53, B=@n54), B=@n55)) {speedX=0.2, speedY=0.2}
  n57 = SmoothStep(Value=MaterialFunctionCall() {materialFunction=&LinearGradient}.VGradient) {constMax=0.55, constMin=0.45}
  n58 = Multiply(
   A=ComponentMask(None=TextureSample(UVs=@n56) {samplerType=SAMPLERTYPE_Grayscale, texture=/Game/Effects/Fort_Effects/Textures/Fire/T_WispyFireNoise_Soft_M}.RGB) {g=true, r=true}
   B=OneMinus(None=@n57)
  )
  n59 = Add(A=@n52, B=Constant() {r=0.5})
  n60 = Divide(A=AppendVector(A=@n53, B=Multiply(A=@n54) {constB=-1}), B=@n55)
  n61 = ComponentMask(None=TextureSample(UVs=Panner(Coordinate=@n59, Speed=@n60) {speedX=0.2, speedY=-0.2}) {samplerType=SAMPLERTYPE_Grayscale, texture=&T_WispyFireNoise_Soft_M}.RGB) {g=true, r=true}
  n62 = Multiply(A=@n61, B=@n57)
  NamedRerouteDeclaration(None=Divide(A=Add(A=@n58, B=@n62), B=Constant() {r=40})) {name=UVDistortionOpposite, nodeColor=<rgba 0 0.7656 1 1>}
  NamedRerouteDeclaration(None=@n61) {name=UVDistortion, nodeColor=<rgba 0.5938 0 1 1>}
  n63 = TextureCoordinate()
  n64 = ComponentMask(None=@n63) {r=true}
  NamedRerouteDeclaration(None=AppendVector(A=Power(Base=@n64) {constExponent=1.3}, B=ComponentMask(None=@n63) {g=true})) {name=PowerUV, nodeColor=<rgba 0.2891 0 1 1>}
  NamedRerouteUsage()
  OneMinus(None=@n30)
  TextureCoordinate()
  TextureCoordinate()
  TextureCoordinate()
  TextureCoordinate()
 ]
}
